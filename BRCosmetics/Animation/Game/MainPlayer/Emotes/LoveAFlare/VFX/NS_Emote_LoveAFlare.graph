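# Niagara system for the LoveAFlare player emote, layering a lit flare with heart effects.
# Flames sit on the flare skeletal mesh and drive sprites, mesh and light with age-timed curves and waveform flicker.
# Hearts rain under gravity and curl noise, and a timed multi-group burst adds flickering glowing pieces late in the emote.
# Glowing hull meshes follow heart prop bones and feed a sparkle emitter through particle reads.
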NiagaraSystem /BRCosmetics/Animation/Game/MainPlayer/Emotes/LoveAFlare/VFX/NS_Emote_LoveAFlare
dependencies {
 usedDataInterfaces [
  &NiagaraDataInterfaceCurve
  &NiagaraDataInterfaceColorCurve
  &NiagaraDataInterfaceSpriteRendererInfo
  &NiagaraDataInterfaceParticleRead
  &NiagaraDataInterfaceArrayColor
  &NiagaraDataInterfaceArrayFloat3
 ]
 usedDynamicInputs [&VectorLength]
 usedModules [
  &self:SystemSpawnScript.SystemScriptSource.SystemScriptGraph.NiagaraNodeAssignment_4.SetVariables_#1
  &self:SystemSpawnScript.SystemScriptSource.SystemScriptGraph.NiagaraNodeAssignment_5.SetVariables_#2
  &SystemState
  &self:SystemSpawnScript.SystemScriptSource.SystemScriptGraph.NiagaraNodeAssignment_0.SetVariables_#3
  &self:SystemSpawnScript.SystemScriptSource.SystemScriptGraph.NiagaraNodeAssignment_1.SetVariables_#4
  &self:SystemSpawnScript.SystemScriptSource.SystemScriptGraph.NiagaraNodeAssignment_2.SetVariables_#5
  &self:SystemSpawnScript.SystemScriptSource.SystemScriptGraph.NiagaraNodeAssignment_3.SetVariables_#6
 ]
}
emitters {
 Flames {
  emitterUpdateScript [
   &EmitterState(Inactive Response=::NewEnumerator1, Life Cycle Mode=::NewEnumerator1, Loop Behavior=::NewEnumerator1, Loop Duration=3.5, UseLoopDelay=false)
   &SpawnBurst_Instantaneous(Use Spawn Probability=false, UseLoopCountLimit=false)
  ]
  particleSpawnScript [
   &InitializeParticle(
    Color=<rgba 4 0.0863 0.02 1>
    Lifetime=&RandomRangeFloat(Minimum=1.5, Maximum=2)
    Mesh Scale Mode=::NewEnumerator3
    Mesh Uniform Scale=0.3
    Sprite Size Mode=::NewEnumerator3
    Uniform Sprite Size=70
    UsePositionOffset=false
    Write Material Random=true
   )
   /SetVariables() {isSetParameters=true}
  ]
  particleUpdateScript [
   &ParticleState(
    Kill Particles When Lifetime Has Elapsed=false
    Let Infinitely Lived Particles Die When Emitter Deactivates=true
    Loop Particles Lifetime=true
   )
   &SkeletalMeshLocation(MaskSpawnGroup=false, Mesh Sampling Type=::NewEnumerator5, Skeletal Mesh=$System.Flare SKM, Socket Sampling Mode=::NewEnumerator5)
   &ScaleSpriteSize(
    Uniform Curve Scale=$System.FlareFlameDuration
    Uniform Curve Sprite Scale=DataInterface() {propertyValues {Curve {keys [
        <at 0 1 RCIM_Cubic>
        {arriveTangent=-0.6622, interpMode=RCIM_Cubic, leaveTangent=-0.6622, tangentMode=RCTM_User, time=0.1, value=0.775}
        {arriveTangent=-0.0208, interpMode=RCIM_Cubic, leaveTangent=-0.0208, tangentMode=RCTM_User, time=0.2, value=0.85}
        {arriveTangent=0.3451, interpMode=RCIM_Cubic, leaveTangent=0.3451, tangentMode=RCTM_User, time=0.3, value=0.875}
        {arriveTangent=-0.0022, interpMode=RCIM_Cubic, leaveTangent=-0.0022, tangentMode=RCTM_User, time=0.4, value=0.95}
        {arriveTangent=-0.0096, interpMode=RCIM_Cubic, leaveTangent=-0.0096, tangentMode=RCTM_User, time=0.5, value=0.875}
        {arriveTangent=-0.037, interpMode=RCIM_Cubic, leaveTangent=-0.037, tangentMode=RCTM_User, time=0.6, value=0.9}
        {arriveTangent=-0.0446, interpMode=RCIM_Cubic, leaveTangent=-0.0446, tangentMode=RCTM_User, time=0.7, value=0.75}
        {arriveTangent=-0.0333, interpMode=RCIM_Cubic, leaveTangent=-0.0333, tangentMode=RCTM_User, time=0.8, value=0.95}
        {arriveTangent=-0.0145, interpMode=RCIM_Cubic, leaveTangent=-0.0145, tangentMode=RCTM_User, time=0.9, value=0.875}
        <at 1 1 RCIM_Cubic>
       ]}}}
   )
   &ScaleSpriteSize(
    Uniform Curve Index=$Emitter.Age
    Uniform Curve Sprite Scale=DataInterface() {propertyValues {Curve {keys [
        {arriveTangent=5.0894, interpMode=RCIM_Cubic, leaveTangent=5.0894, tangentMode=RCTM_User, time=0, value=0.4}
        {arriveTangent=-0.2694, interpMode=RCIM_Cubic, leaveTangent=-0.2694, tangentMode=RCTM_User, time=0.15, value=1.5}
        <at 0.3 1 RCIM_Cubic>
       ]}}}
   ) {module=ScaleSpriteSize001}
   &ScaleMeshSize(Scale Factor=&MakeVector(
     X=$System.FlareFlameDuration
     Y=$System.FlareFlameDuration
     Z=&FloatFromCurve(
      FloatCurve=DataInterface() {propertyValues {Curve {keys [
          {arriveTangent=5.0894, interpMode=RCIM_Cubic, leaveTangent=5.0894, tangentMode=RCTM_User, time=0, value=0.4}
          {arriveTangent=0.2071, interpMode=RCIM_Cubic, leaveTangent=0.2071, tangentMode=RCTM_User, time=0.2, value=1.8}
          {arriveTangent=-0.0762, interpMode=RCIM_Cubic, leaveTangent=-0.0762, time=0.5, value=1}
          {arriveTangent=-0.0727, interpMode=RCIM_Cubic, leaveTangent=-0.0727, time=10.7, value=1}
          <at 11.5 0.2 RCIM_Cubic>
         ]}}}
      CurveIndex=$Emitter.Age
     )
    ))
   /SetVariables(Particles.ColorFlare:LinearColor=<rgba 40 1.0608 0.4733 1>) {isSetParameters=true}
   /SetVariables(Particles.SpriteFlareSize:Vector2f=&Multiply_Vector2D_ByFloat(
     Float=&FloatFromCurve(
      FloatCurve=DataInterface() {propertyValues {Curve {keys [<at 0.23 0.3924>,<at 0.3119 0.6626>,<at 0.4338 0.6715>,<at 0.6 1>,<at 2.5446 1>,<at 3.085 0>]}}}
      CurveIndex=$System.Age
     )
     Vector2D=&Multiply_Vector2D_ByFloat(
      Float=$System.FlareDuration
      Vector2D=&MakeVector2D(
       X=&Multiply_Float(
        A=60
        B=&FloatFromCurve(
         FloatCurve=DataInterface() {propertyValues {Curve {keys [
             <at 0 1 RCIM_Cubic>
             {arriveTangent=0.0166, interpMode=RCIM_Cubic, leaveTangent=0.0166, tangentMode=RCTM_User, time=0.1, value=0.725}
             {arriveTangent=-0.0112, interpMode=RCIM_Cubic, leaveTangent=-0.0112, tangentMode=RCTM_User, time=0.2, value=0.875}
             {arriveTangent=0.0059, interpMode=RCIM_Cubic, leaveTangent=0.0059, tangentMode=RCTM_User, time=0.3, value=0.85}
             {arriveTangent=-0.0022, interpMode=RCIM_Cubic, leaveTangent=-0.0022, tangentMode=RCTM_User, time=0.4, value=0.95}
             {arriveTangent=-0.0096, interpMode=RCIM_Cubic, leaveTangent=-0.0096, tangentMode=RCTM_User, time=0.5, value=0.8125}
             {arriveTangent=-0.037, interpMode=RCIM_Cubic, leaveTangent=-0.037, tangentMode=RCTM_User, time=0.6, value=0.85}
             {arriveTangent=0.0048, interpMode=RCIM_Cubic, leaveTangent=0.0048, tangentMode=RCTM_User, time=0.7, value=0.725}
             {arriveTangent=-0.011, interpMode=RCIM_Cubic, leaveTangent=-0.011, tangentMode=RCTM_User, time=0.8, value=0.95}
             {arriveTangent=-0.0145, interpMode=RCIM_Cubic, leaveTangent=-0.0145, tangentMode=RCTM_User, time=0.9, value=0.875}
             <at 1 1 RCIM_Cubic>
            ]}}}
         Scale Curve=&FloatFromCurve(
          FloatCurve=DataInterface() {propertyValues {Curve {keys [
              {arriveTangent=5.0894, interpMode=RCIM_Cubic, leaveTangent=5.0894, tangentMode=RCTM_User, time=0, value=0.4}
              {arriveTangent=1.2, interpMode=RCIM_Cubic, leaveTangent=1.2, time=0.1, value=2}
              <at 0.5 1 RCIM_Cubic>
             ]}}}
          CurveIndex=$Emitter.Age
         )
        )
       )
       Y=15
      )
     )
    )) {isSetParameters=true}
   /SetVariables(Particles.ColorFlame:LinearColor=&MakeLinearColorFromVectorAndFloat(
     "Vector (RGB)"=&MakeVectorFromLinearColorRGB(Color=<rgba 160 7.6853 1.8248 1>)
     "Float (Alpha)"=&Multiply_Float(
      A=&FloatFromCurve(FloatCurve=DataInterface() {propertyValues {Curve {keys [
           <at 0 1 RCIM_Cubic>
           {arriveTangent=0.0166, interpMode=RCIM_Cubic, leaveTangent=0.0166, tangentMode=RCTM_User, time=0.1, value=0.6}
           {arriveTangent=-0.0112, interpMode=RCIM_Cubic, leaveTangent=-0.0112, tangentMode=RCTM_User, time=0.2, value=0.875}
           {arriveTangent=0.0059, interpMode=RCIM_Cubic, leaveTangent=0.0059, tangentMode=RCTM_User, time=0.3, value=0.7}
           {arriveTangent=-0.0022, interpMode=RCIM_Cubic, leaveTangent=-0.0022, tangentMode=RCTM_User, time=0.4, value=0.95}
           {arriveTangent=-0.0096, interpMode=RCIM_Cubic, leaveTangent=-0.0096, tangentMode=RCTM_User, time=0.5, value=0.75}
           {arriveTangent=-0.037, interpMode=RCIM_Cubic, leaveTangent=-0.037, tangentMode=RCTM_User, time=0.6, value=0.775}
           {arriveTangent=0.0048, interpMode=RCIM_Cubic, leaveTangent=0.0048, tangentMode=RCTM_User, time=0.7, value=0.65}
           {arriveTangent=-0.011, interpMode=RCIM_Cubic, leaveTangent=-0.011, tangentMode=RCTM_User, time=0.8, value=0.9}
           {arriveTangent=-0.0145, interpMode=RCIM_Cubic, leaveTangent=-0.0145, tangentMode=RCTM_User, time=0.9, value=0.825}
           <at 1 1 RCIM_Cubic>
          ]}}})
      B=&FloatFromCurve(
       FloatCurve=DataInterface() {propertyValues {Curve {keys [
           {arriveTangent=5.0894, interpMode=RCIM_Cubic, leaveTangent=5.0894, tangentMode=RCTM_User, time=0, value=0.4}
           {arriveTangent=1.2, interpMode=RCIM_Cubic, leaveTangent=1.2, time=0.2, value=10}
           <at 0.5 1 RCIM_Cubic>
          ]}}}
       CurveIndex=$Emitter.Age
       Scale Curve=$System.FlareFlameDuration
      )
     )
    )) {isSetParameters=true}
   &CameraOffset(Camera Offset Amount=6)
   /SetVariables(Particles.ColorLight:LinearColor=&MakeLinearColorFromVectorAndFloat(
     "Vector (RGB)"=&MakeVectorFromLinearColorRGB(Color=<rgba 18 0.3663 0 1>)
     "Float (Alpha)"=&Multiply_Float(
      A=$System.FlareFlameDuration
      B=&FloatFromCurve(
       FloatCurve=DataInterface() {propertyValues {Curve {keys [<at 0 0>,<at 0.0257 1>,<at 0.0375 0>,<at 0.317 0>,<at 0.3487 1.0012>,<at 0.3634 0>,<at 0.4513 0>,<at 0.458 0.8384>,<at 0.5192 0.1236>,<at 0.6 1>,<at 1 1>,<at 3.3 1>,<at 3.32 0>,<at 3.4154 0>,<at 3.4201 0.9691>,<at 3.528 0.9665>,<at 3.539 0>]}}}
       CurveIndex=$System.Age
      )
     )
    )) {isSetParameters=true}
   &ScaleColor(
    Scale RGB=&VectorFromFloat(Value=&FloatFromCurve(
      FloatCurve=DataInterface() {propertyValues {Curve {keys [
          {arriveTangent=5.0894, interpMode=RCIM_Cubic, leaveTangent=5.0894, tangentMode=RCTM_User, time=0, value=0.4}
          {arriveTangent=2, interpMode=RCIM_Cubic, leaveTangent=2, time=0.1, value=6}
          <at 0.3 1 RCIM_Cubic>
         ]}}}
      CurveIndex=$Emitter.Age
      Scale Curve=$System.FlareFlameDuration
     ))
    ScaleA=true
    ScaleRGB=true
   )
   &DynamicMaterialParameters(
    Index 1 Param 4=&FloatFromCurve(
     FloatCurve=DataInterface() {propertyValues {Curve {keys [
         {arriveTangent=-0.0917, interpMode=RCIM_Cubic, leaveTangent=-0.0917, tangentMode=RCTM_User, time=0, value=1}
         {arriveTangent=-0.031, interpMode=RCIM_Cubic, leaveTangent=-0.031, tangentMode=RCTM_User, time=9, value=0.5}
         <at 11 0.1 RCIM_Cubic>
        ]}}}
     CurveIndex=$Emitter.Age
     Scale Curve=1.5
    )
    Module.Index1Float0Write=true
    Module.Index1Float1Write=true
    Module.Index1Float2Write=true
    Module.Index1Float3Write=true
    Write Parameter Index 0=::NewEnumerator0
    Write Parameter Index 1=::NewEnumerator1
   )
   &ScaleColor(
    Scale Alpha=&Multiply_Float(
     A=&FloatFromCurve(
      FloatCurve=DataInterface() {propertyValues {Curve {keys [<at 0 0>,<at 0.0121 0.5354>,<at 0.0264 0.0362>,<at 0.044 0.1842>,<at 0.0461 -0.024>,<at 0.0858 0.0935>,<at 0.1218 -0.0144>,<at 0.317 0>,<at 0.3271 1.0012>,<at 0.3715 0>,<at 0.4275 -0.0002>,<at 0.4278 0.9045>,<at 0.4581 -0.0064>,<at 0.4871 0.7378>,<at 0.5548 0.5371>,<at 0.6 1>,<at 1 1>,<at 3.3 1>,<at 3.32 0>,<at 3.4154 0>,<at 3.4201 3>,<at 3.528 2>,<at 3.539 0>]}}}
      CurveIndex=$System.Age
     )
     B=&Waveform(
      "Waveform [1]"=::NewEnumerator0
      "[1] Frequency"=&RandomRangeFloat(Minimum=2, Maximum=2.5)
      "[1] Amplitude Min/Max"=<0.35 1>
     )
    )
    ScaleA=true
    ScaleRGB=false
   ) {module=ScaleColor001}
   /SetVariables(Particles.Flame Update:LinearColor=&MultiplyLinearColor(
     Linear Color=$Particles.ColorFlame
     Scale Factor=&MakeVector4_FromLinearColor(LinearColor=&MakeLinearColorFromVectorAndFloat(
       "Vector (RGB)"=<1 1 1>
       "Float (Alpha)"=&FloatFromCurve(
        FloatCurve=DataInterface() {propertyValues {Curve {keys [<at 0.317 0>,<at 0.3438 0.6001>,<at 0.3715 0>,<at 0.4144 0.0924>,<at 0.4597 0.8602>,<at 0.4977 0.1951>,<at 0.6 1>,<at 1 1>,<at 3.0034 0.941>,<at 3.1816 0.0354>,<at 3.2819 0.0118>,<at 3.3163 0.7804>,<at 3.3709 -0.0118>]}}}
        CurveIndex=$System.Age
        Scale Curve=&Waveform(
         "Waveform [1]"=::NewEnumerator0
         "[1] Frequency"=&RandomRangeFloat(Minimum=2, Maximum=2.5)
         "[1] Amplitude Min/Max"=<0.35 1>
        )
       )
      ))
    )) {isSetParameters=true}
   /SetVariables(Particles.Flare Update:LinearColor=&MultiplyLinearColor(
     Linear Color=$Particles.ColorFlare
     Scale Factor=&MakeVector4_FromLinearColor(LinearColor=&MakeLinearColorFromVectorAndFloat(
       "Vector (RGB)"=<1 1 1>
       "Float (Alpha)"=&Multiply_Float(
        A=&FloatFromCurve(
         FloatCurve=DataInterface() {propertyValues {Curve {keys [<at 0.317 0>,<at 0.3271 1.0012>,<at 0.3715 0>,<at 0.4015 0.0064>,<at 0.4122 0.8948>,<at 0.4477 0.431>,<at 0.6 1>,<at 1 1>,<at 3.3 1>,<at 3.32 0>,<at 3.4154 0>,<at 3.4201 2>,<at 3.528 0.9665>,<at 3.539 0>]}}}
         CurveIndex=$System.Age
        )
        B=1
       )
      ))
    )) {isSetParameters=true}
   &ScaleSpriteSize(
    Scale Sprite Size Mode=::NewEnumerator0
    Uniform Scale Factor=&Multiply_Float(
     A=&FloatFromCurve(
      FloatCurve=DataInterface() {propertyValues {Curve {keys [<at 0.23 0.3924>,<at 0.3119 0.6626>,<at 0.4338 0.6715>,<at 0.6 1>,<at 3.271 1>,<at 3.3564 0.6288>,<at 3.3879 4.9545>,<at 3.4668 3.5116>,<at 3.478 0.7514>]}}}
      CurveIndex=$System.Age
     )
     B=&Waveform(
      "Waveform [1]"=::NewEnumerator0
      "[1] Frequency"=&RandomRangeFloat(Minimum=2, Maximum=2.5)
      "[1] Amplitude Min/Max"=<0.75 1>
     )
    )
   ) {module=ScaleSpriteSize002}
  ]
  rendererClasses [
   &NiagaraSpriteRendererProperties
   &NiagaraLightRendererProperties
   &NiagaraMeshRendererProperties
  ]
  renderers [
   {rendererClass=&NiagaraSpriteRendererProperties}
   {rendererClass=&NiagaraSpriteRendererProperties, rendererIndex=1}
   {rendererClass=&NiagaraLightRendererProperties, rendererIndex=2}
   {rendererClass=&NiagaraMeshRendererProperties, rendererIndex=3}
   {rendererClass=&NiagaraSpriteRendererProperties, rendererIndex=4}
  ]
 }
 FlareDepsawnSparks {rendererClasses [], renderers []}
 FlareLightingSparks {rendererClasses [], renderers []}
 Hearts {
  emitterUpdateScript [
   &EmitterState(Life Cycle Mode=::NewEnumerator1, Loop Duration=2.6, UseLoopDelay=false)
   &SpawnRate(
    SpawnRate=&Multiply_Float(
     A=45
     B=&FloatFromCurve(
      FloatCurve=DataInterface() {propertyValues {Curve {keys [
          {arriveTangent=9.1514, interpMode=RCIM_Cubic, leaveTangent=9.1514, tangentMode=RCTM_User, time=3.9, value=0}
          {arriveTangent=0.2369, interpMode=RCIM_Cubic, leaveTangent=0.2369, tangentMode=RCTM_User, time=4.2126, value=1}
          {arriveTangent=-0.6472, interpMode=RCIM_Cubic, leaveTangent=-0.6472, tangentMode=RCTM_User, time=5.4585, value=0.6566}
          <at 5.9318 0 RCIM_Cubic>
         ]}}}
      CurveIndex=$Engine.System.Age
     )
    )
    Use Spawn Probability=false
   )
  ]
  particleSpawnScript [
   &PartitionParticles(Number Of Partitions=3, Partition Mode:ENiagara_PartitionMode_v2=::NewEnumerator5)
   /SetVariables(Particles.VisibilityTag=$Particles.PartitionParticles.Partition) {isSetParameters=true}
   &InitializeParticle(
    Lifetime Max=0.8
    Lifetime Min=0.5
    Lifetime Mode=::NewEnumerator1
    Mass Mode=::NewEnumerator3
    Mesh Scale Mode=::NewEnumerator1
    Mesh Uniform Scale Max=1.1
    Mesh Uniform Scale Min=0.3
    Sprite Rotation Angle=&UniformAOrBFloat(
     A=0
     B=&RandomRangeFloat(Minimum=-40, Maximum=40)
     Evaluation Type=::NewEnumerator0
     Recalculate Random Each Loop=true
     Override Seed=false
    )
    Sprite Rotation Mode=::NewEnumerator0
    Sprite Size Mode=::NewEnumerator1
    Uniform Sprite Size Max=20
    Uniform Sprite Size Min=10
    UsePositionOffset=false
    Write Material Random=true
   )
   &ShapeLocation(
    Handle Radius=50
    MaskSpawnGroup=false
    Offset=<30 0 220>
    Offset Mode=::NewEnumerator0
    Override Seed=false
    Shape Primitive=::NewEnumerator3
    UseOwnerScale=true
   )
   /SetVariables(Particles.Mesh Cam Offset=8) {isSetParameters=true}
   &InitialMeshOrientation(
    Mesh Orientation Mode=::NewEnumerator2
    Orientation Vector=&RandomRangeVector(Minimum=<-0.5 -0.9 0>, Maximum=<0.5 0.9 0>)
    Use Rotation Vector=false
   )
   /SetVariables(Particles.Sprite Cam Offset:NiagaraFloat=&RandomRangeFloat(Minimum=-2, Maximum=-12)) {isSetParameters=true}
  ]
  particleUpdateScript [
   &ParticleState()
   &SpriteRotationRate(Rotation Rate=&MakeCustomFloatFromBool(
     Bool=&SetBoolByFloatComparison(A=0.33, B=&RandomRangeFloat(Minimum=0, Maximum=1))
     True Float=&RandomRangeFloat(Minimum=-10, Maximum=10)
     False Float=&RandomRangeFloat(Minimum=-80, Maximum=80)
    ))
   &ScaleMeshSize(Scale Factor=&VectorFromFloat(Value=&FloatFromCurve(FloatCurve=DataInterface() {propertyValues {Curve {
         defaultValue=0
         keys [
          {interpMode=RCIM_Cubic, tangentMode=RCTM_User, time=0, value=0}
          {interpMode=RCIM_Cubic, tangentMode=RCTM_User, time=0.4632, value=1}
          {arriveTangent=-1.8628, interpMode=RCIM_Cubic, leaveTangent=-1.8628, time=0.7355, value=0.7701}
          {tangentMode=RCTM_User, time=1, value=0}
         ]
        }}})))
   &UpdateMeshOrientation(Rotation Rate=&RandomRangeFloat(Minimum=0.5, Maximum=2), Rotation Vector=<0 0 1>)
   &ScaleSpriteSize(Uniform Curve Sprite Scale=DataInterface() {propertyValues {Curve {
       defaultValue=0
       keys [
        {interpMode=RCIM_Cubic, tangentMode=RCTM_User, time=0, value=0}
        {interpMode=RCIM_Cubic, tangentMode=RCTM_User, time=0.4808, value=1}
        {arriveTangent=-1.9262, interpMode=RCIM_Cubic, leaveTangent=-1.9262, time=0.7355, value=0.7701}
        {tangentMode=RCTM_User, time=1, value=0}
       ]
      }}})
   &GravityForce(Gravity=&RandomRangeVector(Minimum=<0 0 -600>, Maximum=<0 0 -980>))
   &Drag(Drag=&RandomRangeFloat(Minimum=0.25, Maximum=0.5), Use Linear Drag=true, Use Rotational Drag=false)
   &ScaleColor(
    Scale Alpha=&FloatFromCurve(FloatCurve=DataInterface() {propertyValues {Curve {keys [
         <at -0.0046 0 RCIM_Cubic>
         {arriveTangent=0.0019, interpMode=RCIM_Cubic, leaveTangent=0.0019, time=0.2089, value=1}
         <at 1.0011 0.002 RCIM_Cubic>
        ]}}})
    ScaleA=true
    ScaleRGB=true
   )
   &CurlNoiseForce(
    Bias Noise=false
    Noise Frequency=&RandomRangeFloat(Minimum=25, Maximum=40)
    Noise Strength=&RandomRangeFloat(Minimum=30, Maximum=100)
    Pan Noise=false
   )
   &SolveForcesAndVelocity(Clamp Velocity=false, Limit Acceleration=false)
   /SetVariables(Particles.Hull Mesh Scale:Vector3f=&Multiply_Vector(A=$Particles.Scale, B=&VectorFromFloat(Value=1.15))) {isSetParameters=true}
   /SetVariables(
    Particles.Blue Glow Color:LinearColor=&MakeLinearColorFromVectorAndFloat(
     "Vector (RGB)"=&MakeVectorFromLinearColorRGB(Color=<rgba 0.002 0.0761 1 1>)
     "Float (Alpha)"=&FloatFromCurve(FloatCurve=DataInterface() {propertyValues {Curve {keys [
          <at -0.0046 0 RCIM_Cubic>
          {arriveTangent=0.0019, interpMode=RCIM_Cubic, leaveTangent=0.0019, time=0.2089, value=1}
          <at 1.0011 0.002 RCIM_Cubic>
         ]}}})
    )
    Particles.Mesh Heart Glow Size:Vector2f=&Multiply_Vector2D_ByFloat(
     Float=&MakeFloatFromVector(Vector=$Particles.Hull Mesh Scale, Channel=::NewEnumerator0)
     Vector2D=&Vector2DFromFloat(Value=75)
    )
    Particles.Red Glow Color:LinearColor=&MakeLinearColorFromVectorAndFloat(
     "Vector (RGB)"=&MakeVectorFromLinearColorRGB(Color=<rgba 1 0.0189 0 1>)
     "Float (Alpha)"=&FloatFromCurve(FloatCurve=DataInterface() {propertyValues {Curve {keys [
          <at -0.0046 0 RCIM_Cubic>
          {arriveTangent=0.0019, interpMode=RCIM_Cubic, leaveTangent=0.0019, time=0.2089, value=1}
          <at 1.0011 0.002 RCIM_Cubic>
         ]}}})
    )
    Particles.Sprite Glow Size:Vector2f=&Multiply_Vector2D_ByFloat(Float=2.5, Vector2D=$Particles.SpriteSize)
   ) {isSetParameters=true}
   /SetVariables(Particles.Mesh Hearts Color:LinearColor=&MakeLinearColorFromVectorAndFloat(
     "Vector (RGB)"=<1 1 1>
     "Float (Alpha)"=&FloatFromCurve(FloatCurve=DataInterface() {propertyValues {Curve {keys [
          <at -0.0046 0 RCIM_Cubic>
          {arriveTangent=0.0019, interpMode=RCIM_Cubic, leaveTangent=0.0019, time=0.2089, value=1}
          {arriveTangent=-2.7322, interpMode=RCIM_Cubic, leaveTangent=-2.7322, tangentMode=RCTM_User, time=1.0011, value=0.002}
         ]}}})
    )) {isSetParameters=true}
  ]
  rendererClasses [&NiagaraSpriteRendererProperties, &NiagaraMeshRendererProperties]
  renderers [
   {rendererClass=&NiagaraSpriteRendererProperties}
   {rendererClass=&NiagaraSpriteRendererProperties, rendererIndex=1}
   {rendererClass=&NiagaraMeshRendererProperties, rendererIndex=2}
   {rendererClass=&NiagaraMeshRendererProperties, rendererIndex=3}
   {rendererClass=&NiagaraSpriteRendererProperties, rendererIndex=4}
   {rendererClass=&NiagaraSpriteRendererProperties, rendererIndex=5}
   {rendererClass=&NiagaraSpriteRendererProperties, rendererIndex=6}
  ]
 }
 Hearts_Depawn_Burst {
  emitterUpdateScript [
   &EmitterState(Life Cycle Mode=::NewEnumerator1, Loop Behavior=::NewEnumerator1, Loop Duration=9, UseLoopDelay=false)
   &SpawnBurst_Instantaneous(
    Spawn Count=&RandomRangeInt(Minimum=1, Maximum=2)
    Spawn Time=7.45
    Use Spawn Probability=false
    UseLoopCountLimit=false
   )
   &SpawnBurst_Instantaneous(
    Spawn Count=&RandomRangeInt(Minimum=1, Maximum=2)
    Spawn Group=1
    Spawn Time=7.45
    Use Spawn Probability=false
    UseLoopCountLimit=false
   ) {module=SpawnBurst_Instantaneous001}
   &SpawnBurst_Instantaneous(Spawn Count=5, Spawn Group=2, Spawn Time=7.45, Use Spawn Probability=false, UseLoopCountLimit=false) {module=SpawnBurst_Instantaneous002}
   &SpawnBurst_Instantaneous(Spawn Count=17, Spawn Group=3, Spawn Time=7.45, Use Spawn Probability=false, UseLoopCountLimit=false) {module=SpawnBurst_Instantaneous003}
   &SpawnBurst_Instantaneous(Spawn Group=4, Spawn Time=7.45, Use Spawn Probability=false, UseLoopCountLimit=false) {module=SpawnBurst_Instantaneous004}
   &SpawnBurst_Instantaneous(Spawn Group=5, Spawn Time=7.45, Use Spawn Probability=false, UseLoopCountLimit=false) {module=SpawnBurst_Instantaneous005}
   &SpawnBurst_Instantaneous(Spawn Group=6, Spawn Time=7.45, Use Spawn Probability=false, UseLoopCountLimit=false) {module=SpawnBurst_Instantaneous006}
  ]
  particleSpawnScript [
   /SetVariables(Particles.VisibilityTag=&SpawnGroup()) {isSetParameters=true}
   &InitializeParticle(
    Color=&MergeHelper_ColorSwitch(
     0=<rgba 1 1 1 1>
     1=<rgba 1 1 1 1>
     2=&SelectLinearColorFromArray(
      Array Sampling Mode=::NewEnumerator0
      Direct Array Index=?
      Color Selection Array=DataInterface() {propertyValues {ColorData [<rgba 1 0.4126 0.5495 1>, <rgba 0.3193 0.6926 1 1>, <rgba 1 1 1 1>]}}
      Random Seed=0
     )
     3=<rgba 1 1 1 1>
     4=<rgba 10 7.8021 7.8021 1>
     5=<rgba 1 0.0744 0.1874 1>
     6=<rgba 1 1 1 1>
     Index=&SpawnGroup()
     Num=::NewEnumerator5
    )
    Lifetime=&MergeHelper_FloatSwitch(
     0=&MakeCustomFloatFromBool(
      Bool=&SetBoolByFloatComparison(A=0.3, B=&RandomRangeFloat(Minimum=0, Maximum=1))
      True Float=0.65
      False Float=&RandomRangeFloat(Minimum=0.35, Maximum=0.47)
     )
     1=&MakeCustomFloatFromBool(
      Bool=&SetBoolByFloatComparison(A=0.3, B=&RandomRangeFloat(Minimum=0, Maximum=1))
      True Float=0.65
      False Float=&RandomRangeFloat(Minimum=0.35, Maximum=0.47)
     )
     2=&MakeCustomFloatFromBool(
      Bool=&SetBoolByFloatComparison(A=0.3, B=&RandomRangeFloat(Minimum=0, Maximum=1))
      True Float=0.7
      False Float=&RandomRangeFloat(Minimum=0.27, Maximum=0.46)
     )
     3=&MakeCustomFloatFromBool(
      Bool=&SetBoolByFloatComparison(A=0.25, B=&RandomRangeFloat(Minimum=0, Maximum=1))
      True Float=0.8
      False Float=&RandomRangeFloat(Minimum=0.3, Maximum=0.63)
     )
     4=0.05
     5=0.18
     6=0.44
     Index=&SpawnGroup()
     Num=::NewEnumerator5
    )
    Position Offset=/Niagara/DynamicInputs/Helpers/MaskPosition(
     Mask Source:NiagaraInt32=0
     Mask Parameter:NiagaraInt32=?
     Mask:NiagaraInt32=5
     Masked Position:NiagaraPosition=<0 8 0>
     Passthrough Position:NiagaraPosition=<0 0 0>
    )
    Sprite Rotation Mode=::NewEnumerator2
    Sprite Size Mode=::NewEnumerator3
    Uniform Sprite Size=&MergeHelper_FloatSwitch(
     0=&RandomRangeFloat(Minimum=5, Maximum=13)
     1=&RandomRangeFloat(Minimum=5, Maximum=13)
     2=&RandomRangeFloat(Minimum=4, Maximum=9)
     3=&MakeCustomFloatFromBool(
      Bool=&SetBoolByFloatComparison(A=0.25, B=&RandomRangeFloat(Minimum=0, Maximum=1))
      True Float=&RandomRangeFloat(Minimum=18, Maximum=25)
      False Float=&RandomRangeFloat(Minimum=5, Maximum=7)
     )
     4=70
     5=57
     6=17
     Index=&SpawnGroup()
     Num=::NewEnumerator5
    )
    UsePositionOffset=true
    Write Material Random=true
   )
   &CameraOffset(Camera Offset Amount=&MaskFloat(
     Mask Source=0
     Mask Parameter=?
     Mask=5
     Masked Float=0
     Passthrough Float=&RandomRangeFloat(Minimum=2, Maximum=5)
    ))
   &ShapeLocation(
    MaskSpawnGroup=false
    Non Uniform Scale=<4.5 1.5 3.5>
    Offset=<30 40 125>
    Offset Mode=::NewEnumerator0
    Override Seed=false
    Sphere Radius=&MaskFloat(
     Mask Source=0
     Mask Parameter=?
     Mask=4
     Masked Float=0
     Passthrough Float=&MaskFloat(Mask Source=0, Mask Parameter=?, Mask=5, Masked Float=0, Passthrough Float=5)
    )
    UseOwnerScale=true
   )
   &AddVelocity(
    Cone Angle=60
    Cone Axis=<0 0.2 1>
    Module.UseConeFalloff=false
    Module.UseVelDistribution=false
    Override Seed=false
    Velocity Mode=::NewEnumerator2
    Velocity Speed=&MergeHelper_FloatSwitch(
     0=&RandomRangeFloat(Minimum=120, Maximum=145)
     1=&RandomRangeFloat(Minimum=120, Maximum=145)
     2=&RandomRangeFloat(Minimum=60, Maximum=160)
     3=&RandomRangeFloat(Minimum=60, Maximum=160)
     4=0
     5=0
     Index=&SpawnGroup()
     Num=::NewEnumerator4
    )
   ) {module=AddVelocity001}
   &AddVelocity(Velocity=&MaskVector(
     Mask Source=0
     Mask Parameter=?
     Mask=4
     Masked Vector=<0 0 0>
     Passthrough Vector=&MaskVector(
      Mask Source=0
      Mask Parameter=?
      Mask=5
      Masked Vector=<0 0 0>
      Passthrough Vector=&MaskVector(
       Mask Source=0
       Mask Parameter=?
       Mask=6
       Masked Vector=<0 0 0>
       Passthrough Vector=&MakeCustomVectorFromBool(
        Bool=&SetBoolByFloatComparison(A=0.33, B=&RandomRangeFloat(Minimum=0, Maximum=1))
        True Vector=&SelectVectorFromArray(
         Array Sampling Mode=::NewEnumerator0
         Direct Array Index=?
         Vector Selection Array=DataInterface() {propertyValues {FloatData [<50 0 0>, <-50 1 0>]}}
         Random Seed=0
        )
        False Vector=&RandomRangeVector(Minimum=<-22 0 0>, Maximum=<22 0 0>)
       )
      )
     )
    ))
  ]
  particleUpdateScript [
   &ParticleState()
   &ScaleSpriteSize(
    Scale Sprite Size Mode=::NewEnumerator0
    Uniform Scale Factor=&MergeHelper_FloatSwitch(
     0=&FloatFromCurve(FloatCurve=DataInterface() {propertyValues {Curve {
         defaultValue=1
         keys [
          {arriveTangent=45.6757, interpMode=RCIM_Cubic, leaveTangent=45.6757, tangentMode=RCTM_User, time=0, value=0}
          {interpMode=RCIM_Cubic, tangentMode=RCTM_User, time=0.06, value=1}
          {arriveTangent=-1.0638, interpMode=RCIM_Cubic, leaveTangent=-1.0638, time=0.4979, value=0.8991}
          {interpMode=RCIM_Cubic, tangentMode=RCTM_User, time=1, value=0}
         ]
        }}})
     1=&FloatFromCurve(FloatCurve=DataInterface() {propertyValues {Curve {
         defaultValue=1
         keys [
          {arriveTangent=45.6757, interpMode=RCIM_Cubic, leaveTangent=45.6757, tangentMode=RCTM_User, time=0, value=0}
          {interpMode=RCIM_Cubic, tangentMode=RCTM_User, time=0.06, value=1}
          {arriveTangent=-1.0638, interpMode=RCIM_Cubic, leaveTangent=-1.0638, time=0.4979, value=0.8991}
          {interpMode=RCIM_Cubic, tangentMode=RCTM_User, time=1, value=0}
         ]
        }}})
     2=&FloatFromCurve(FloatCurve=DataInterface() {propertyValues {Curve {
         defaultValue=1
         keys [
          {interpMode=RCIM_Cubic, tangentMode=RCTM_User, time=0, value=1}
          {arriveTangent=-1.2892, interpMode=RCIM_Cubic, leaveTangent=-1.2892, tangentMode=RCTM_User, time=1, value=0}
         ]
        }}})
     3=&FloatFromCurve(FloatCurve=DataInterface() {propertyValues {Curve {
         defaultValue=1
         keys [
          {interpMode=RCIM_Cubic, tangentMode=RCTM_User, time=0.4063, value=1}
          {arriveTangent=-4.4648, interpMode=RCIM_Cubic, leaveTangent=-4.4648, tangentMode=RCTM_User, time=1, value=0}
         ]
        }}})
     4=&FloatFromCurve(FloatCurve=DataInterface() {propertyValues {Curve {
         defaultValue=1
         keys [
          {interpMode=RCIM_Cubic, tangentMode=RCTM_User, time=0.826, value=1}
          {arriveTangent=-14.7449, interpMode=RCIM_Cubic, leaveTangent=-14.7449, tangentMode=RCTM_User, time=1, value=0}
         ]
        }}})
     5=&FloatFromCurve(
      FloatCurve=DataInterface() {propertyValues {Curve {
         defaultValue=1
         keys [
          {arriveTangent=7.3284, interpMode=RCIM_Cubic, leaveTangent=7.3284, tangentMode=RCTM_User, time=-0.001, value=0.25}
          {arriveTangent=1.4622, interpMode=RCIM_Cubic, leaveTangent=1.4622, tangentMode=RCTM_User, time=0.1442, value=0.8794}
          {arriveTangent=0.0224, interpMode=RCIM_Cubic, leaveTangent=0.0224, tangentMode=RCTM_User, time=1.0031, value=1.5}
         ]
        }}}
      Scale Curve=0.8
     )
     6=&FloatFromCurve(FloatCurve=DataInterface() {propertyValues {Curve {
         defaultValue=1
         keys [
          {arriveTangent=45.6757, interpMode=RCIM_Cubic, leaveTangent=45.6757, tangentMode=RCTM_User, time=0, value=0}
          {interpMode=RCIM_Cubic, tangentMode=RCTM_User, time=0.06, value=1}
          {arriveTangent=-1.0638, interpMode=RCIM_Cubic, leaveTangent=-1.0638, time=0.2381, value=0.956}
          {interpMode=RCIM_Cubic, tangentMode=RCTM_User, time=1, value=0}
         ]
        }}})
     Index=&SpawnGroup()
     Num=::NewEnumerator5
    )
   )
   /SetVariables(
    Particles.Alpha Flicker:NiagaraFloat=&Waveform(
     "Waveform [1]"=::NewEnumerator0
     "[1] Frequency"=&RandomRangeFloat(Minimum=5, Maximum=7)
     "[1] Amplitude Min/Max"=<0.25 1>
    )
    Particles.Flicker:LinearColor=&ColorFromCurve(ColorCurve=DataInterface() {propertyValues {
       AlphaCurve {keys []}
       BlueCurve {keys [<at 0 1 RCIM_Cubic>,<at 0.0265 10>,<at 0.0493 1>,<at 0.1673 1>,<at 0.1991 0.3365>,<at 0.2371 1>,<at 0.3069 1>,<at 0.3323 10>,<at 0.3564 1>,<at 0.463 1>,<at 0.4896 0.3031>,<at 0.5138 1>,<at 0.6343 1>,<at 0.6584 15>,<at 0.6838 1>,<at 0.8031 1>,<at 0.8386 0.2865>,<at 0.8703 1>,<at 0.9756 1>]}
       GreenCurve {keys [<at 0 1 RCIM_Cubic>,<at 0.0265 10>,<at 0.0493 1>,<at 0.1673 1>,<at 0.1991 0.3365>,<at 0.2371 1>,<at 0.3069 1>,<at 0.3323 10>,<at 0.3564 1>,<at 0.463 1>,<at 0.4896 0.3031>,<at 0.5138 1>,<at 0.6343 1>,<at 0.6584 15>,<at 0.6838 1>,<at 0.8031 1>,<at 0.8386 0.2865>,<at 0.8703 1>,<at 0.9756 1>]}
       RedCurve {keys [<at 0 1 RCIM_Cubic>,<at 0.0265 10>,<at 0.0493 1>,<at 0.1673 1>,<at 0.1991 0.3365>,<at 0.2371 1>,<at 0.3069 1>,<at 0.3323 10>,<at 0.3564 1>,<at 0.463 1>,<at 0.4896 0.3031>,<at 0.5138 1>,<at 0.6343 1>,<at 0.6584 15>,<at 0.6838 1>,<at 0.8031 1>,<at 0.8386 0.2865>,<at 0.8703 1>,<at 0.9756 1>]}
      }})
   ) {isSetParameters=true}
   &ScaleColor(
    Scale Alpha=&MergeHelper_FloatSwitch(
     0=&FloatFromCurve(FloatCurve=DataInterface() {propertyValues {Curve {keys [<at -0.0046 1.0113 RCIM_Cubic>,<at 1.0011 0.002 RCIM_Cubic>]}}})
     1=&FloatFromCurve(FloatCurve=DataInterface() {propertyValues {Curve {keys [<at -0.0046 1.0113 RCIM_Cubic>,<at 1.0011 0.002 RCIM_Cubic>]}}})
     2=&FloatFromCurve(FloatCurve=DataInterface() {propertyValues {Curve {keys [<at -0.0046 1.0113 RCIM_Cubic>,<at 1.0011 0.002 RCIM_Cubic>]}}}, Scale Curve=1.5)
     3=&FloatFromCurve(
      FloatCurve=DataInterface() {propertyValues {Curve {keys [<at -0.0046 1.0113 RCIM_Cubic>,<at 1.0011 0.002 RCIM_Cubic>]}}}
      Scale Curve=&Multiply_Float(A=1, B=$Particles.Alpha Flicker)
     )
     4=&FloatFromCurve(FloatCurve=DataInterface() {propertyValues {Curve {keys [<at -0.0046 1.0113 RCIM_Cubic>,<at 1.0011 0.002 RCIM_Cubic>]}}})
     5=&FloatFromCurve(
      FloatCurve=DataInterface() {propertyValues {Curve {keys [
          <at 0.1274 1 RCIM_Cubic>
          {arriveTangent=-1.4321, interpMode=RCIM_Cubic, leaveTangent=-1.4321, tangentMode=RCTM_User, time=0.2146, value=0.5739}
          <at 1.0011 0.002 RCIM_Cubic>
         ]}}}
      Scale Curve=1.2
     )
     6=&FloatFromCurve(FloatCurve=DataInterface() {propertyValues {Curve {keys [<at -0.0046 1.0113 RCIM_Cubic>,<at 1.0011 0.002 RCIM_Cubic>]}}})
     Index=&SpawnGroup()
     Num=::NewEnumerator5
    )
    Scale RGB=&MakeVectorFromLinearColorRGB(Color=&MergeHelper_ColorSwitch(
      0=$Particles.Flicker
      1=$Particles.Flicker
      2=$Particles.Flicker
      3=$Particles.Flicker
      4=<rgba 1 1 1 1>
      5=<rgba 1 1 1 1>
      6=<rgba 1 1 1 1>
      Index=&SpawnGroup()
      Num=::NewEnumerator5
     ))
    ScaleA=true
    ScaleRGB=true
   )
   &Drag(Drag=&RandomRangeFloat(Minimum=0.7, Maximum=1.4), Use Linear Drag=true, Use Rotational Drag=true)
   &GravityForce(Gravity=&Multiply_VectorByFloat(
     Vector=&MergeHelper_VectorSwitch(
      0=<0 0 -400>
      1=<0 0 -400>
      2=<0 0 -600>
      3=<0 0 -800>
      4=<0 0 0>
      5=<0 0 0>
      6=<0 0 -980>
      Index=&SpawnGroup()
      Num=::NewEnumerator5
     )
     Float=&FloatFromCurve(FloatCurve=DataInterface() {propertyValues {Curve {defaultValue=1, keys [{tangentMode=RCTM_User, time=0.25, value=0}, {tangentMode=RCTM_User, time=0.8, value=1}]}}})
    ))
   &CurlNoiseForce(
    Bias Noise=false
    Noise Strength=&Multiply_Float(
     A=&MergeHelper_FloatSwitch(0=200, 1=200, 2=200, 3=200, 4=0, 5=0, Index=&SpawnGroup(), Num=::NewEnumerator4)
     B=&FloatFromCurve(FloatCurve=DataInterface() {propertyValues {Curve {defaultValue=1, keys [{tangentMode=RCTM_User, time=0.2, value=0}, {tangentMode=RCTM_User, time=0.7, value=1}]}}})
    )
    Pan Noise=false
   )
   &SolveForcesAndVelocity(Clamp Velocity=false, Limit Acceleration=false)
   /SetVariables() {isSetParameters=true}
   /SetVariables(Particles.Set Rotation Range:NiagaraFloat=&RandomRangeFloat(Minimum=-12, Maximum=12)) {isSetParameters=true}
   /SetVariables(
    Particles.Blue Glow Color:LinearColor=&MultiplyLinearColor(
     Linear Color=<rgba 0.002 0.0761 1 1>
     Scale Factor=&MakeVector4_FromLinearColor(LinearColor=$Particles.Color)
    )
    Particles.Heart Glow Size:Vector2f=&Multiply_Vector2D_ByFloat(Float=2.25, Vector2D=$Particles.SpriteSize)
    Particles.Red Glow Color:LinearColor=&MultiplyLinearColor(Linear Color=<rgba 1 0.0189 0 1>, Scale Factor=&MakeVector4_FromLinearColor(LinearColor=$Particles.Color))
    Particles.Sparkle Glow Color:LinearColor=&MultiplyLinearColor(
     Linear Color=&MakeLinearColorFromVectorAndFloat(
      "Vector (RGB)"=&MakeVectorFromLinearColorRGB(Color=&SelectLinearColorFromArray(
        Array Sampling Mode=::NewEnumerator0
        Direct Array Index=?
        Color Selection Array=DataInterface() {propertyValues {ColorData [<rgba 1 0.0189 0 1>, <rgba 0.002 0.0761 1 1>]}}
        Random Seed=0
       ))
      "Float (Alpha)"=&RandomRangeFloat(Minimum=0.75, Maximum=0.9)
     )
     Scale Factor=&MakeVector4_FromLinearColor(LinearColor=$Particles.Color)
    )
    Particles.Sparkle Glow Size:Vector2f=&Multiply_Vector2D_ByFloat(Float=1.8, Vector2D=$Particles.SpriteSize)
   ) {isSetParameters=true}
   /SetVariables(Particles.Hot Outline Color:LinearColor=&MakeLinearColorFromVectorAndFloat(
     "Vector (RGB)"=&MakeVectorFromLinearColorRGB(Color=<rgba 15 4.0799 4.8541 1>)
     "Float (Alpha)"=&FloatFromCurve(
      FloatCurve=DataInterface() {propertyValues {Curve {keys [
          <at 0.0966 1 RCIM_Cubic>
          {arriveTangent=-1.4321, interpMode=RCIM_Cubic, leaveTangent=-1.4321, tangentMode=RCTM_User, time=0.1992, value=0.5739}
          <at 1.0011 0.002 RCIM_Cubic>
         ]}}}
      Scale Curve=0.8
     )
    )) {isSetParameters=true}
  ]
  rendererClasses [&NiagaraSpriteRendererProperties]
  renderers [
   {rendererClass=&NiagaraSpriteRendererProperties}
   {rendererClass=&NiagaraSpriteRendererProperties, rendererIndex=1}
   {rendererClass=&NiagaraSpriteRendererProperties, rendererIndex=2}
   {rendererClass=&NiagaraSpriteRendererProperties, rendererIndex=3}
   {rendererClass=&NiagaraSpriteRendererProperties, rendererIndex=4}
   {rendererClass=&NiagaraSpriteRendererProperties, rendererIndex=5}
   {rendererClass=&NiagaraSpriteRendererProperties, rendererIndex=6}
   {rendererClass=&NiagaraSpriteRendererProperties, rendererIndex=7}
   {rendererClass=&NiagaraSpriteRendererProperties, rendererIndex=8}
   {rendererClass=&NiagaraSpriteRendererProperties, rendererIndex=9}
   {rendererClass=&NiagaraSpriteRendererProperties, rendererIndex=10}
  ]
 }
 Prop_Heart_Hull {
  emitterUpdateScript [
   &EmitterState()
   &SpawnBurst_Instantaneous(Spawn Count=2, Use Spawn Probability=false, UseLoopCountLimit=false)
  ]
  particleSpawnScript [
   &PartitionParticles()
   &InitializeParticle(Color Mode=::NewEnumerator0, Lifetime=8.089, Mesh Scale Mode=::NewEnumerator3, UsePositionOffset=false, Write Material Random=true)
   &CameraOffset(Camera Offset Amount=-0.5)
  ]
  particleUpdateScript [
   &ParticleState()
   &SkeletalMeshLocation(
    Bone / Socket Index=$Particles.PartitionParticles.Partition
    Bone Sampling Mode=::NewEnumerator4
    MaskSpawnGroup=false
    Skeletal Mesh=$System.Heart 1 SKM
   )
   &ScaleMeshSize(Scale Factor=&Multiply_VectorByFloat(Vector=$Particles.SkeletalMeshLocation.SampledScale, Float=1.1))
   /SetVariables(
    Particles.Glow Color=&MakeLinearColorFromVectorAndFloat(
     "Vector (RGB)"=&MakeVectorFromLinearColorRGB(Color=<rgba 1 0.0189 0 1>)
     "Float (Alpha)"=&FloatFromCurve(FloatCurve=DataInterface() {propertyValues {Curve {
         defaultValue=1
         keys [
          {interpMode=RCIM_Cubic, tangentMode=RCTM_User, time=0.6898, value=1}
          {arriveTangent=-6.8887, interpMode=RCIM_Cubic, leaveTangent=-6.8887, time=0.7474, value=0.3411}
          {interpMode=RCIM_Cubic, tangentMode=RCTM_User, time=0.835, value=0}
         ]
        }}})
    )
    Particles.Glow Size=&Multiply_Vector2D_ByFloat(
     Float=&MakeFloatFromVector(Vector=$Particles.Scale, Channel=::NewEnumerator0)
     Vector2D=&Vector2DFromFloat(Value=75)
    )
   ) {isSetParameters=true}
  ]
  rendererClasses [&NiagaraMeshRendererProperties, &NiagaraSpriteRendererProperties]
  renderers [
   {rendererClass=&NiagaraMeshRendererProperties, rendererIndex=0}
   {rendererClass=&NiagaraSpriteRendererProperties, rendererIndex=1}
  ]
 }
 Prop_Sparkles {
  emitterUpdateScript [
   &EmitterState(Life Cycle Mode=::NewEnumerator1, Loop Duration=2, UseLoopDelay=false)
   &SpawnParticlesFromOtherEmitter(
    Attribute Reader=DataInterface() {propertyValues {EmitterBinding {bindingMode=Other, emitterName=Prop_Heart_Hull}}}
    Spawn Rate=30
    Use Maximum Age=false
    Use Minimum Age=false
   )
  ]
  particleSpawnScript [
   &InitializeParticle(
    Lifetime Max=0.5
    Lifetime Min=0.2
    Lifetime Mode=::NewEnumerator1
    Mass Mode=::NewEnumerator3
    Sprite Rotation Mode=::NewEnumerator2
    Sprite Size Mode=::NewEnumerator3
    Uniform Sprite Size=&MakeCustomFloatFromBool(
     Bool=&SetBoolByFloatComparison(A=0.25, B=&RandomRangeFloat(Minimum=0, Maximum=1))
     True Float=&RandomRangeFloat(Minimum=18, Maximum=25)
     False Float=&RandomRangeFloat(Minimum=5, Maximum=7)
    )
    UsePositionOffset=false
    Write Material Random=true
   )
   &AddVelocity(Velocity=&RandomRangeVector(Minimum=<0 0 10>, Maximum=<0 0 25>))
   &SampleParticlesFromOtherEmitter(
    Applied Alpha Scale=0.6
    Applied Velocity Scale=0.3
    Apply Sampled RGB=false
    Color Sampling=::NewEnumerator0
    Maximum Sampled Particle Age=6.3
    Minimum Sampled Particle Age=5.1
    Module.MaxSampledAgeCutoff=true
    Module.MinSampledAgeCutoff=true
    Reject Particle By Sampled Age=true
    Velocity Sampling=::NewEnumerator0
   )
   &ShapeLocation(MaskSpawnGroup=false, Non Uniform Scale=<1.2 0.6 1>, Override Seed=false, Sphere Radius=8, UseOwnerScale=true)
  ]
  particleUpdateScript [
   &ParticleState()
   &GravityForce(Gravity=&RandomRangeVector(Minimum=<0 0 0>, Maximum=<0 0 12>))
   &Drag(Drag=&RandomRangeFloat(Minimum=0.25, Maximum=1), Use Linear Drag=true, Use Rotational Drag=false)
   &ScaleColor(
    Scale Alpha=&FloatFromCurve(
     FloatCurve=DataInterface() {propertyValues {Curve {keys [<at -0.0046 1.0113 RCIM_Cubic>,<at 1.0011 0.002 RCIM_Cubic>]}}}
     Scale Curve=&Multiply_Float(
      A=1
      B=&Waveform(
       "Waveform [1]"=::NewEnumerator0
       "[1] Frequency"=&RandomRangeFloat(Minimum=5, Maximum=7)
       "[1] Amplitude Min/Max"=<0.25 1>
      )
     )
    )
    Scale RGB=&MakeVectorFromLinearColorRGB(Color=&ColorFromCurve(ColorCurve=DataInterface() {propertyValues {
        AlphaCurve {keys []}
        BlueCurve {keys [<at 0 1 RCIM_Cubic>,<at 0.0601 1>,<at 0.0853 10>,<at 0.1113 1>,<at 0.1607 1>,<at 0.1834 0.2948>,<at 0.2047 1>,<at 0.3807 1>,<at 0.4058 0.3198>,<at 0.431 1>,<at 0.4931 1>,<at 0.519 20>,<at 0.5443 1>,<at 0.5919 1>,<at 0.6195 0.2906>,<at 0.6434 1>,<at 0.739 0.3781>,<at 0.7788 1>,<at 0.8684 1>,<at 0.9074 0.4156>,<at 0.9313 1>,<at 0.9702 1>]}
        GreenCurve {keys [<at 0 1 RCIM_Cubic>,<at 0.0601 1>,<at 0.0853 10>,<at 0.1113 1>,<at 0.1607 1>,<at 0.1834 0.2948>,<at 0.2047 1>,<at 0.3807 1>,<at 0.4058 0.3198>,<at 0.431 1>,<at 0.4931 1>,<at 0.519 20>,<at 0.5443 1>,<at 0.5919 1>,<at 0.6195 0.2906>,<at 0.6434 1>,<at 0.739 0.3781>,<at 0.7788 1>,<at 0.8684 1>,<at 0.9074 0.4156>,<at 0.9313 1>,<at 0.9702 1>]}
        RedCurve {keys [<at 0 1 RCIM_Cubic>,<at 0.0601 1>,<at 0.0853 10>,<at 0.1113 1>,<at 0.1607 1>,<at 0.1834 0.2948>,<at 0.2047 1>,<at 0.3807 1>,<at 0.4058 0.3198>,<at 0.431 1>,<at 0.4931 1>,<at 0.519 20>,<at 0.5443 1>,<at 0.5919 1>,<at 0.6195 0.2906>,<at 0.6434 1>,<at 0.739 0.3781>,<at 0.7788 1>,<at 0.8684 1>,<at 0.9074 0.4156>,<at 0.9313 1>,<at 0.9702 1>]}
       }}))
    ScaleA=true
    ScaleRGB=true
   )
   &CurlNoiseForce(
    Bias Noise=false
    Noise Frequency=&RandomRangeFloat(Minimum=40, Maximum=70)
    Noise Strength=&RandomRangeFloat(Minimum=70, Maximum=150)
    Pan Noise=false
   )
   &SolveForcesAndVelocity(Clamp Velocity=false, Limit Acceleration=false)
   /SetVariables(
    Particles.Glow Color=&MultiplyLinearColor(
     Linear Color=&MakeLinearColorFromVectorAndFloat(
      "Vector (RGB)"=&MakeVectorFromLinearColorRGB(Color=&SelectLinearColorFromArray(
        Array Sampling Mode=::NewEnumerator0
        Direct Array Index=?
        Color Selection Array=DataInterface() {propertyValues {ColorData [<rgba 1 0.0189 0 1>, <rgba 0.002 0.0761 1 1>]}}
        Random Seed=0
       ))
      "Float (Alpha)"=1
     )
     Scale Factor=&MakeVector4_FromLinearColor(LinearColor=$Particles.Color)
    )
    Particles.Glow Size=&Multiply_Vector2D_ByFloat(Float=1.25, Vector2D=$Particles.SpriteSize)
   ) {isSetParameters=true}
  ]
  rendererClasses [&NiagaraSpriteRendererProperties]
  renderers [
   {rendererClass=&NiagaraSpriteRendererProperties}
   {rendererClass=&NiagaraSpriteRendererProperties, rendererIndex=1}
  ]
 }
 Smoke {
  emitterUpdateScript [
   &EmitterState(Life Cycle Mode=::NewEnumerator1, Loop Behavior=::NewEnumerator1, Loop Duration=3.5, UseLoopDelay=false)
   &SpawnRate(
    SpawnRate=&Multiply_Float(
     A=60
     B=&FloatFromCurve(
      FloatCurve=DataInterface() {propertyValues {Curve {keys [
          {arriveTangent=4.9615, interpMode=RCIM_Cubic, leaveTangent=4.9615, tangentMode=RCTM_User, time=0.67, value=0}
          {arriveTangentWeight=0.1817, interpMode=RCIM_Cubic, leaveTangentWeight=0.6939, tangentMode=RCTM_User, time=1.1743, value=1}
          {arriveTangent=-0.3007, interpMode=RCIM_Cubic, leaveTangent=-0.3007, time=3.6026, value=0.8456}
          <at 4.5 0 RCIM_Cubic>
         ]}}}
      CurveIndex=$Engine.System.Age
     )
    )
    Use Spawn Probability=false
   )
  ]
  particleSpawnScript [
   &PartitionParticles(Partition Mode:ENiagara_PartitionMode_v2=::NewEnumerator5)
   /SetVariables(Particles.VisibilityTag=$Particles.PartitionParticles.Partition) {isSetParameters=true}
   &InitializeParticle(
    Color Maximum=<rgba 1 1 1 0.23>
    Color Minimum=<rgba 1 1 1 0.15>
    Color Mode=::NewEnumerator2
    Lifetime Max=1.3
    Lifetime Min=0.9
    Lifetime Mode=::NewEnumerator1
    Mass Mode=::NewEnumerator3
    Sprite Rotation Mode=::NewEnumerator2
    Sprite Size Mode=::NewEnumerator1
    Uniform Sprite Size Max=30
    Uniform Sprite Size Min=18
    UsePositionOffset=false
    Write Material Random=true
   )
   &ShapeLocation(MaskSpawnGroup=false, Override Seed=false, Sphere Radius=8, UseOwnerScale=true)
   &AddVelocity(
    Cone Angle=32
    Cone Axis=<0 0 1>
    Distribution Along Cone Axis=0.25
    Module.UseConeFalloff=true
    Module.UseVelDistribution=true
    Override Seed=false
    Speed Falloff From Cone Axis=0.5
    Velocity Mode=::NewEnumerator2
    Velocity Speed=&RandomRangeFloat(Minimum=80, Maximum=200)
   )
   &SkeletalMeshLocation(MaskSpawnGroup=false, Mesh Sampling Type=::NewEnumerator5, Skeletal Mesh=$System.Flare SKM, Socket Sampling Mode=::NewEnumerator5)
  ]
  particleUpdateScript [
   &ParticleState()
   &Drag(Drag=0.25, Use Linear Drag=true, Use Rotational Drag=false)
   &ScaleSpriteSize(Scale Factor=&Vector2DFromFloat(Value=&Multiply_Float(
      A=&FloatFromCurve(
       FloatCurve=DataInterface() {propertyValues {Curve {keys [
           {arriveTangent=2.5043, interpMode=RCIM_Cubic, leaveTangent=2.5043, tangentMode=RCTM_User, time=0.0016, value=0.2}
           {arriveTangent=0.2677, interpMode=RCIM_Cubic, leaveTangent=0.2677, tangentMode=RCTM_User, time=0.8464, value=1.0206}
           {arriveTangent=-236.4009, interpMode=RCIM_Cubic, leaveTangent=-236.4009, tangentMode=RCTM_User, time=0.85, value=0}
          ]}}}
       Scale Curve=2.25
      )
      B=&FloatFromCurve(
       FloatCurve=DataInterface() {propertyValues {Curve {
          defaultValue=1
          keys [
           {arriveTangent=4.0943, interpMode=RCIM_Cubic, leaveTangent=4.0943, tangentMode=RCTM_User, time=0.2462, value=1}
           {tangentMode=RCTM_User, time=1, value=2}
          ]
         }}}
       CurveIndex=$System.NormalizedLoopAge
      )
     )))
   &ScaleColor(
    Scale Alpha=&Multiply_Float(
     A=&FloatFromCurve(FloatCurve=DataInterface() {propertyValues {Curve {keys [
          {arriveTangent=45.1311, interpMode=RCIM_Cubic, leaveTangent=45.1311, tangentMode=RCTM_User, time=0, value=0}
          <at 0.05 1 RCIM_Cubic>
          <at 0.85 0>
         ]}}})
     B=&FloatFromCurve(
      FloatCurve=DataInterface() {propertyValues {Curve {
         defaultValue=1
         keys [{tangentMode=RCTM_User, time=0.2707, value=1}, <at 0.3648 0.4319>, <at 0.6284 0.2211>, {tangentMode=RCTM_User, time=1, value=0}]
        }}}
      CurveIndex=$System.NormalizedLoopAge
     )
    )
    Scale RGB=&MakeVectorFromLinearColorRGB(Color=&MultiplyLinearColor(
      Linear Color=&ColorFromCurve(ColorCurve=DataInterface() {propertyValues {
         AlphaCurve {keys []}
         BlueCurve {keys [
           <at 0.0094 0.0003>
           {arriveTangent=3.4677, interpMode=RCIM_Cubic, leaveTangent=3.4677, time=0.0694, value=0.3457}
           <at 0.1212 0.3882>
           <at 0.1799 0.3852>
           <at 0.2617 0.4702>
           <at 0.2999 0.4834>
           <at 0.3487 0.5674>
           <at 0.423 0.5781>
           <at 0.6544 0.5448>
           <at 0.9101 0.2448>
          ]}
         GreenCurve {keys [
           <at 0.0094 0.0307>
           {arriveTangent=4.306, interpMode=RCIM_Cubic, leaveTangent=4.306, time=0.0694, value=0.5367}
           <at 0.1212 0.5123>
           <at 0.1799 0.5761>
           <at 0.2617 0.4219>
           <at 0.2999 0.5022>
           <at 0.3487 0.574>
           <at 0.423 0.5048>
           <at 0.6544 0.4752>
           <at 0.9101 0.2041>
          ]}
         RedCurve {keys [
           <at 0.0094 20>
           {arriveTangent=-134.0934, interpMode=RCIM_Cubic, leaveTangent=-134.0934, time=0.0694, value=5}
           <at 0.1212 5>
           <at 0.1799 1.631>
           <at 0.2617 0.7365>
           <at 0.2999 0.574>
           <at 0.3487 0.5673>
           <at 0.423 0.4405>
           <at 0.6544 0.4171>
           <at 0.9101 0.2221>
          ]}
        }})
      Scale Factor=&MakeVector4_FromLinearColor(LinearColor=&ColorFromCurve(
        ColorCurve=DataInterface() {propertyValues {
          AlphaCurve {keys []}
          BlueCurve {keys [
            <at 0.252 1 RCIM_Cubic>
            {arriveTangent=-3.3964, interpMode=RCIM_Cubic, leaveTangent=-3.3964, time=0.4554, value=0.0161}
            <at 0.5425 0.0134>
            <at 0.6356 0.0281>
           ]}
          GreenCurve {keys [
            <at 0.252 1 RCIM_Cubic>
            {arriveTangent=-3.436, interpMode=RCIM_Cubic, leaveTangent=-3.436, time=0.4554, value=0.0227}
            <at 0.5425 0.0018>
            <at 0.6356 0.0021>
           ]}
          RedCurve {keys [
            <at 0.252 1 RCIM_Cubic>
            {arriveTangent=-3.3455, interpMode=RCIM_Cubic, leaveTangent=-3.3455, time=0.4554, value=0.1656}
            <at 0.5425 0.0281>
            <at 0.6356 0.0094>
           ]}
         }}
        CurveIndex=$System.NormalizedLoopAge
       ))
     ))
    ScaleA=true
    ScaleRGB=true
   )
   &SolveForcesAndVelocity(Clamp Velocity=false, Limit Acceleration=false)
   &SubUVAnimation(
    Output Alternate Renderer Binding=true
    "Sprite Renderer "=DataInterface() {propertyValues {SpriteRenderer=&self:Smoke_0.NiagaraSpriteRendererProperties_1}}
    Start Frame Offset:NiagaraInt32=&RandomRangeInt(Minimum=0, Maximum=5)
    UseEndFrame=false
    UseStartFrame=false
   )
   &DynamicMaterialParameters(
    Index 0 Param 1=&FloatFromCurve(FloatCurve=DataInterface() {propertyValues {Curve {keys [<at 0 0 RCIM_Cubic>, {arriveTangent=0.4364, tangentMode=RCTM_User, time=1.0011, value=0.2}]}}})
    Index 0 Param 2=&FloatFromCurve(FloatCurve=DataInterface() {propertyValues {Curve {keys [
         {arriveTangent=0.0453, interpMode=RCIM_Cubic, tangentMode=RCTM_Break, time=0, value=-1}
         {arriveTangent=0.0604, interpMode=RCIM_Cubic, leaveTangent=0.0604, tangentMode=RCTM_User, time=1, value=-1}
        ]}}})
    Index 0 Param 3=&FloatFromCurve(FloatCurve=DataInterface() {propertyValues {Curve {keys [{interpMode=RCIM_Cubic, leaveTangent=4.6779, time=0.2851, value=0}]}}})
    Index 1 Param 1=&FloatFromCurve(
     FloatCurve=DataInterface() {propertyValues {Curve {keys [<at 0 0>,<at 0.25 1>]}}}
     Scale Curve=&RandomRangeFloat(Minimum=2, Maximum=1)
    )
    Index 1 Param 2=&FloatFromCurve(
     FloatCurve=DataInterface() {propertyValues {Curve {keys [
         {interpMode=RCIM_Cubic, leaveTangent=3.7422, tangentMode=RCTM_Break, time=0, value=0.2}
         {interpMode=RCIM_Cubic, leaveTangent=0.9066, time=1, value=2}
        ]}}}
     Scale Curve=&RandomRangeFloat(Minimum=0.5, Maximum=1)
    )
    Index 1 Param 3=&FloatFromCurve(FloatCurve=DataInterface() {propertyValues {Curve {keys [
         {interpMode=RCIM_Cubic, leaveTangent=2.8445, tangentMode=RCTM_Break, time=0, value=0.2}
         {interpMode=RCIM_Cubic, leaveTangent=63.2978, time=0.25, value=1}
        ]}}})
    Index 1 Param 4=&FloatFromCurve(
     FloatCurve=DataInterface() {propertyValues {Curve {keys [<at 0 1>,<at 1 0.8>]}}}
     Scale Curve=&RandomRangeFloat(Minimum=0.5, Maximum=1)
    )
    Index 2 Param 1=&FloatFromCurve(FloatCurve=DataInterface() {propertyValues {Curve {keys [{interpMode=RCIM_Cubic, leaveTangent=3.2309, tangentMode=RCTM_Break, time=0.1, value=0}, <at 1 1>]}}})
    Index 2 Param 2=$Particles.MaterialRandom
    Index 2 Param 3=&FloatFromCurve(FloatCurve=DataInterface() {propertyValues {Curve {keys [
         {interpMode=RCIM_Cubic, tangentMode=RCTM_Break, time=0, value=1}
         {interpMode=RCIM_Cubic, leaveTangent=11.1233, tangentMode=RCTM_Break, time=1, value=2}
        ]}}})
    Index 2 Param 4=0.75
    Module.Param0WriteEnabled=true
    Module.Param0WriteEnabled1=true
    Module.Param0WriteEnabled2=true
    Module.Param1WriteEnabled=false
    Module.Param1WriteEnabled1=true
    Module.Param1WriteEnabled2=true
    Module.Param2WriteEnabled=true
    Module.Param2WriteEnabled1=true
    Module.Param2WriteEnabled2=true
    Module.Param3WriteEnabled=true
    Module.Param3WriteEnabled1=true
    Module.Param3WriteEnabled2=true
    Write Parameter Index 1=true
    Write Parameter Index 2=true
   )
  ]
  rendererClasses [&NiagaraSpriteRendererProperties]
  renderers [
   {rendererClass=&NiagaraSpriteRendererProperties}
   {rendererClass=&NiagaraSpriteRendererProperties, rendererIndex=1}
  ]
 }
 Smoke_Ribbon {
  emitterUpdateScript [
   &EmitterState(Life Cycle Mode=::NewEnumerator1, Loop Behavior=::NewEnumerator1, Loop Duration=3.5, UseLoopDelay=false)
   &SpawnRate(
    SpawnRate=&Multiply_Float(
     A=100
     B=&FloatFromCurve(
      FloatCurve=DataInterface() {propertyValues {Curve {keys [
          {arriveTangent=4.9615, interpMode=RCIM_Cubic, leaveTangent=4.9615, tangentMode=RCTM_User, time=0.67, value=0}
          {arriveTangentWeight=0.1817, interpMode=RCIM_Cubic, leaveTangentWeight=0.6939, tangentMode=RCTM_User, time=1.1743, value=1}
          {arriveTangent=-0.3007, interpMode=RCIM_Cubic, leaveTangent=-0.3007, time=3.6026, value=0.8456}
          <at 4.5 0 RCIM_Cubic>
         ]}}}
      CurveIndex=$Engine.System.Age
     )
    )
    Use Spawn Probability=false
   )
  ]
  particleSpawnScript [
   &InitializeParticle(
    Lifetime Max=1.1
    Lifetime Min=0.9
    Lifetime Mode=::NewEnumerator1
    Ribbon Width=40
    Ribbon Width Mode=::NewEnumerator1
    UsePositionOffset=false
    Write Material Random=true
   )
   &AddVelocity(Velocity=<0 0 250>)
   &ApplyInitialForces()
   &SkeletalMeshLocation(MaskSpawnGroup=false, Mesh Sampling Type=::NewEnumerator5, Skeletal Mesh=$System.Flare SKM, Socket Sampling Mode=::NewEnumerator5)
  ]
  particleUpdateScript [
   &ParticleState()
   &Drag(Drag=0.5, Use Linear Drag=true, Use Rotational Drag=true) {enabled=false}
   &SolveForcesAndVelocity(Clamp Velocity=false, Limit Acceleration=false)
   &DynamicMaterialParameters(
    Index 0 Param 1=-1
    Index 0 Param 4=1
    Module.Index0Float0Write=true
    Module.Index0Float1Write=true
    Module.Index0Float2Write=true
    Module.Index0Float3Write=true
   )
   &ScaleRibbonWidth(Ribbon Width Scale=&FloatFromCurve(FloatCurve=DataInterface() {propertyValues {Curve {
        defaultValue=1
        keys [
         <at 0.0042 0.4456 RCIM_Cubic>
         {arriveTangent=8.1687, interpMode=RCIM_Cubic, leaveTangent=8.1687, tangentMode=RCTM_User, time=0.1003, value=0.8631}
         {arriveTangent=5.0857, interpMode=RCIM_Cubic, leaveTangent=5.0857, tangentMode=RCTM_User, time=0.2394, value=1.611}
         <at 1.0002 2.5758 RCIM_Cubic>
        ]
       }}}))
   &ScaleColor(
    Scale Alpha=&Multiply_Float(
     A=&FloatFromCurve(FloatCurve=DataInterface() {propertyValues {Curve {
         defaultValue=1
         keys [
          {interpMode=RCIM_Cubic, tangentMode=RCTM_User, time=0.0221, value=1}
          {arriveTangent=-1.9935, interpMode=RCIM_Cubic, leaveTangent=-1.9935, tangentMode=RCTM_User, time=0.1574, value=0.3565}
          {interpMode=RCIM_Cubic, tangentMode=RCTM_User, time=0.7024, value=0.0129}
         ]
        }}})
     B=&FloatFromCurve(
      FloatCurve=DataInterface() {propertyValues {Curve {keys [
          {arriveTangent=4.9615, interpMode=RCIM_Cubic, leaveTangent=4.9615, tangentMode=RCTM_User, time=0.67, value=0}
          {arriveTangentWeight=0.1681, interpMode=RCIM_Cubic, leaveTangentWeight=0.7352, tangentMode=RCTM_User, time=1.1743, value=1}
          {arriveTangentWeight=0.7925, interpMode=RCIM_Cubic, leaveTangentWeight=0.1009, tangentMode=RCTM_User, time=3.3799, value=1}
          <at 3.6609 0 RCIM_Cubic>
         ]}}}
      CurveIndex=$Engine.System.Age
     )
    )
    Scale RGB=&MakeVectorFromLinearColorRGB(Color=&ColorFromCurve(ColorCurve=DataInterface() {propertyValues {
        AlphaCurve {keys []}
        BlueCurve {keys [
          <at 0.0028 0.1477>
          {arriveTangent=2.1472, interpMode=RCIM_Cubic, leaveTangent=2.1472, time=0.0267, value=0.224}
          <at 0.1061 0.3695>
          <at 0.1924 0.5772>
          <at 0.2694 0.5674>
          <at 0.3542 0.5781>
          <at 0.6976 0.5448>
          <at 0.9101 0.2448>
         ]}
        GreenCurve {keys [
          <at 0.0028 0.142>
          {arriveTangent=2.1102, interpMode=RCIM_Cubic, leaveTangent=2.1102, time=0.0267, value=0.2856}
          <at 0.1061 0.3599>
          <at 0.1924 0.6505>
          <at 0.2694 0.574>
          <at 0.3542 0.5048>
          <at 0.6976 0.4752>
          <at 0.9101 0.2041>
         ]}
        RedCurve {keys [<at 0.0028 1>,<at 0.0267 1 RCIM_Cubic>,<at 0.1061 1>,<at 0.1924 0.901>,<at 0.2694 0.5673>,<at 0.3542 0.4405>,<at 0.6976 0.4171>,<at 0.9101 0.2221>]}
       }}))
    ScaleA=true
    ScaleRGB=true
   )
  ]
  rendererClasses [&NiagaraRibbonRendererProperties]
  renderers [{rendererClass=&NiagaraRibbonRendererProperties}]
 }
 Sparkles {
  emitterUpdateScript [
   &EmitterState(Life Cycle Mode=::NewEnumerator1, Loop Duration=2, UseLoopDelay=false)
   &SpawnParticlesFromOtherEmitter(
    Attribute Reader=DataInterface() {propertyValues {EmitterBinding {bindingMode=Other, emitterName=Hearts}}}
    Spawn Rate=30
    Use Maximum Age=false
    Use Minimum Age=false
   )
  ]
  particleSpawnScript [
   &InitializeParticle(
    Lifetime Max=0.5
    Lifetime Min=0.2
    Lifetime Mode=::NewEnumerator1
    Mass Mode=::NewEnumerator3
    Sprite Rotation Mode=::NewEnumerator2
    Sprite Size Mode=::NewEnumerator3
    Uniform Sprite Size=&MakeCustomFloatFromBool(
     Bool=&SetBoolByFloatComparison(A=0.25, B=&RandomRangeFloat(Minimum=0, Maximum=1))
     True Float=&RandomRangeFloat(Minimum=18, Maximum=25)
     False Float=&RandomRangeFloat(Minimum=5, Maximum=7)
    )
    UsePositionOffset=false
    Write Material Random=true
   )
   &AddVelocity(Velocity=&RandomRangeVector(Minimum=<0 0 10>, Maximum=<0 0 25>))
   &SampleParticlesFromOtherEmitter(
    Applied Alpha Scale=0.6
    Applied Velocity Scale=0.3
    Apply Sampled RGB=false
    Color Sampling=::NewEnumerator0
    Maximum Sampled Particle Age=0.6
    Module.MaxSampledAgeCutoff=true
    Module.MinSampledAgeCutoff=false
    Reject Particle By Sampled Age=true
    Velocity Sampling=::NewEnumerator0
   )
   &ShapeLocation(MaskSpawnGroup=false, Non Uniform Scale=<1.2 0.6 1>, Override Seed=false, Sphere Radius=8, UseOwnerScale=true)
  ]
  particleUpdateScript [
   &ParticleState()
   &GravityForce(Gravity=&RandomRangeVector(Minimum=<0 0 0>, Maximum=<0 0 12>))
   &Drag(Drag=&RandomRangeFloat(Minimum=0.25, Maximum=1), Use Linear Drag=true, Use Rotational Drag=false)
   &ScaleColor(
    Scale Alpha=&FloatFromCurve(
     FloatCurve=DataInterface() {propertyValues {Curve {keys [<at 0 1 RCIM_Cubic>,<at 1.0011 0.002 RCIM_Cubic>]}}}
     Scale Curve=&Multiply_Float(
      A=1
      B=&Waveform(
       "Waveform [1]"=::NewEnumerator0
       "[1] Frequency"=&RandomRangeFloat(Minimum=5, Maximum=7)
       "[1] Amplitude Min/Max"=<0.35 1>
      )
     )
    )
    Scale RGB=&MakeVectorFromLinearColorRGB(Color=&ColorFromCurve(ColorCurve=DataInterface() {propertyValues {
        AlphaCurve {keys []}
        BlueCurve {keys [<at 0 1 RCIM_Cubic>,<at 0.0601 1>,<at 0.0853 10>,<at 0.1113 1>,<at 0.1607 1>,<at 0.1834 0.2948>,<at 0.2047 1>,<at 0.3807 1>,<at 0.4058 0.3198>,<at 0.431 1>,<at 0.4931 1>,<at 0.519 20>,<at 0.5443 1>,<at 0.5919 1>,<at 0.6195 0.2906>,<at 0.6434 1>,<at 0.739 0.3781>,<at 0.7788 1>,<at 0.8684 1>,<at 0.9074 0.4156>,<at 0.9313 1>,<at 0.9702 1>]}
        GreenCurve {keys [<at 0 1 RCIM_Cubic>,<at 0.0601 1>,<at 0.0853 10>,<at 0.1113 1>,<at 0.1607 1>,<at 0.1834 0.2948>,<at 0.2047 1>,<at 0.3807 1>,<at 0.4058 0.3198>,<at 0.431 1>,<at 0.4931 1>,<at 0.519 20>,<at 0.5443 1>,<at 0.5919 1>,<at 0.6195 0.2906>,<at 0.6434 1>,<at 0.739 0.3781>,<at 0.7788 1>,<at 0.8684 1>,<at 0.9074 0.4156>,<at 0.9313 1>,<at 0.9702 1>]}
        RedCurve {keys [<at 0 1 RCIM_Cubic>,<at 0.0601 1>,<at 0.0853 10>,<at 0.1113 1>,<at 0.1607 1>,<at 0.1834 0.2948>,<at 0.2047 1>,<at 0.3807 1>,<at 0.4058 0.3198>,<at 0.431 1>,<at 0.4931 1>,<at 0.519 20>,<at 0.5443 1>,<at 0.5919 1>,<at 0.6195 0.2906>,<at 0.6434 1>,<at 0.739 0.3781>,<at 0.7788 1>,<at 0.8684 1>,<at 0.9074 0.4156>,<at 0.9313 1>,<at 0.9702 1>]}
       }}))
    ScaleA=true
    ScaleRGB=true
   )
   &CurlNoiseForce(
    Bias Noise=false
    Noise Frequency=&RandomRangeFloat(Minimum=40, Maximum=70)
    Noise Strength=&RandomRangeFloat(Minimum=70, Maximum=150)
    Pan Noise=false
   )
   &SolveForcesAndVelocity(Clamp Velocity=false, Limit Acceleration=false)
   /SetVariables(
    Particles.Glow Color=&MultiplyLinearColor(
     Linear Color=&MakeLinearColorFromVectorAndFloat(
      "Vector (RGB)"=&MakeVectorFromLinearColorRGB(Color=&SelectLinearColorFromArray(
        Array Sampling Mode=::NewEnumerator0
        Direct Array Index=?
        Color Selection Array=DataInterface() {propertyValues {ColorData [<rgba 1 0.0189 0 1>, <rgba 0.002 0.0761 1 1>]}}
        Random Seed=0
       ))
      "Float (Alpha)"=1
     )
     Scale Factor=&MakeVector4_FromLinearColor(LinearColor=$Particles.Color)
    )
    Particles.Glow Size=&Multiply_Vector2D_ByFloat(Float=1.25, Vector2D=$Particles.SpriteSize)
   ) {isSetParameters=true}
  ]
  rendererClasses [&NiagaraSpriteRendererProperties]
  renderers [
   {rendererClass=&NiagaraSpriteRendererProperties}
   {rendererClass=&NiagaraSpriteRendererProperties, rendererIndex=1}
  ]
 }
}
summary {
 systemName=NS_Emote_LoveAFlare
 emitters [
  {emitterName=Hearts, rendererClasses [&NiagaraSpriteRendererProperties, &NiagaraMeshRendererProperties]}
  {emitterName=Smoke, rendererClasses [&NiagaraSpriteRendererProperties]}
  {
   emitterName=Flames
   rendererClasses [
    &NiagaraSpriteRendererProperties
    &NiagaraLightRendererProperties
    &NiagaraMeshRendererProperties
   ]
  }
  {emitterName=Sparkles, rendererClasses [&NiagaraSpriteRendererProperties]}
  {
   emitterName=Prop_Heart_Hull
   rendererClasses [&NiagaraMeshRendererProperties, &NiagaraSpriteRendererProperties]
  }
  {emitterName=Prop_Sparkles, rendererClasses [&NiagaraSpriteRendererProperties]}
  {emitterName=Hearts_Depawn_Burst, rendererClasses [&NiagaraSpriteRendererProperties]}
  {emitterName=Smoke_Ribbon, rendererClasses [&NiagaraRibbonRendererProperties]}
  {emitterName=FlareLightingSparks, rendererClasses []}
  {emitterName=FlareDepsawnSparks, rendererClasses []}
 ]
 userVariables [
  User.bIsFrontEnd {type=NiagaraBool}
  User.bIsFrontEndPreview {type=NiagaraBool}
  User.Loop {type=NiagaraFloat, defaultValue=3.1469, description=""}
  User.RibbonIDOffset {type=NiagaraInt32}
  User.SkeletalMesh {
   type=NiagaraDataInterfaceSkeletalMesh
   defaultValue {
    struct=NiagaraExt_VariableValue_DataInterface
    value {dataInterface=&self:NiagaraDataInterfaceSkeletalMesh_0}
   }
  }
  User.SocketNumber {type=NiagaraInt32}
 ]
}
